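# Conversation database for the PiteousKicks NPC, defining the greeting spoken when a player starts talking to the character.
# A single entry point feeds one speech node with a default line and contextual alternatives.
# Priority curves rank first-ever-meeting and outfit-specific reactions, while general lines check the NPC's match conversation count.
# The speech node then links out to the shared NPC-library quests and basic services conversation entries.
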ConversationDatabase /NPCLibrary/NPCs/PiteousKicks/NPCLibrary_Conversation_PiteousKicks
entry_points [{tag=Conversation.Entry.PiteousKicks, destinations [#1]}]
nodes [
 ConversationEntryPointNode_7 {
  id=#1
  class=ConversationEntryPointNode
  outputs [#2]
  properties {entryTag {tagName=Conversation.Entry.PiteousKicks}}
 }
 Speech_C_4 {
  id=#2
  class=Speech_C
  outputs [#3, #4]
  properties {
   nodeName=""
   generalConfig {
    defaultMessage=Make it worth my time.
    contextualMessages [
     {
      message=You've heard right. I'm back.
      contextRequirements [{requirement {class=FortControllerRequirement_IsFirstEverConversationWithNPC}}]
      priority {curve {rowName=Tandem.Speech.Priority.FirstEverInteraction}}
     }
     {
      message=It's... very nice to see you again.
      contextRequirements [{requirement {
         class=FortControllerRequirement_HasCID
         properties {allowedCIds [/BRCosmetics/Athena/Items/Cosmetics/Characters/CID_A_034_Athena_Commando_F_CatBurglar]}
        }}]
      priority {curve {rowName=Tandem.Speech.Priority.OutfitReaction}}
     }
     {
      message=I see you've grown beyond my shadow. I hope you haven't left me behind.
      contextRequirements [{requirement {
         class=FortControllerRequirement_HasCID
         properties {allowedCIds [/BRCosmetics/Athena/Items/Cosmetics/Characters/CID_770_Athena_Commando_F_MechanicalEngineer]}
        }}]
      priority {curve {rowName=Tandem.Speech.Priority.OutfitReaction}}
     }
     {
      message="Good, you're back. Let's get to work."
      contextRequirements [{requirement {
         class=FortControllerRequirement_HasCID
         properties {allowedCIds [
           &CID_692_Athena_Commando_M_HenchmanTough
           /BRCosmetics/Athena/Items/Cosmetics/Characters/CID_690_Athena_Commando_F_Photographer
           /BRCosmetics/Athena/Items/Cosmetics/Characters/CID_691_Athena_Commando_F_TNTina
           &CID_693_Athena_Commando_M_BuffCat
          ]}
        }}]
      priority {curve {rowName=Tandem.Speech.Priority.OutfitReaction}}
     }
     {
      message=Make it worth my time.
      contextRequirements [{requirement {class=FortControllerRequirement_CheckNPCMatchConversationCount, properties {nPCUniqueIdTag {tagName=AISpawnerData.Type.Tandem.PiteousKicks}}}}]
      priority {curve {rowName=None}}
     }
     {
      message=You have my attention. Don't waste it.
      contextRequirements [{requirement {class=FortControllerRequirement_CheckNPCMatchConversationCount, properties {nPCUniqueIdTag {tagName=AISpawnerData.Type.Tandem.PiteousKicks}}}}]
      priority {curve {rowName=None}}
     }
     {
      message=This Island was always too small for my ambitions.
      contextRequirements [{requirement {class=FortControllerRequirement_CheckNPCMatchConversationCount, properties {nPCUniqueIdTag {tagName=AISpawnerData.Type.Tandem.PiteousKicks}}}}]
      priority {curve {rowName=None}}
     }
     {
      message=I don't fear these gods. No power is absolute.
      contextRequirements [{requirement {class=FortControllerRequirement_CheckNPCMatchConversationCount, properties {nPCUniqueIdTag {tagName=AISpawnerData.Type.Tandem.PiteousKicks}}}}]
      priority {curve {rowName=None}}
     }
     {
      message=Hades thought he could keep me trapped. He would have been the first.
      contextRequirements [{requirement {class=FortControllerRequirement_CheckNPCMatchConversationCount, properties {nPCUniqueIdTag {tagName=AISpawnerData.Type.Tandem.PiteousKicks}}}}]
      priority {curve {rowName=None}}
     }
     {
      message=It seems I've missed quite a bit in my time away.
      contextRequirements [{requirement {class=FortControllerRequirement_CheckNPCMatchConversationCount, properties {nPCUniqueIdTag {tagName=AISpawnerData.Type.Tandem.PiteousKicks}}}}]
      priority {curve {rowName=None}}
     }
     {
      message=Nothing lasts forever. Except gold.
      contextRequirements [{requirement {class=FortControllerRequirement_CheckNPCMatchConversationCount, properties {nPCUniqueIdTag {tagName=AISpawnerData.Type.Tandem.PiteousKicks}}}}]
      priority {curve {rowName=None}}
     }
     {
      message="In the myths they tell of me, they never quite get the details right."
      contextRequirements [{requirement {class=FortControllerRequirement_CheckNPCMatchConversationCount, properties {nPCUniqueIdTag {tagName=AISpawnerData.Type.Tandem.PiteousKicks}}}}]
      priority {curve {rowName=None}}
     }
    ]
   }
  }
 }
 ConversationLinkNode_0 {id=#4, class=ConversationLinkNode, properties {remoteEntryTag {tagName=Conversation.Entry.BasicServices}}}
 ConversationLinkNode_1 {id=#3, class=ConversationLinkNode, properties {remoteEntryTag {tagName=Conversation.Entry.NPCLibrary.Quests}}}
]
